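# Idle Niagara effect for the Angler pickaxe that adds a looping, color-shifting light to the weapon.
# A single Light emitter bursts one particle on its first loop only, keeps it alive, and re-samples its position from User.SkeletalMesh each update.
# System normalized loop age drives a multicolor gradient, boosted for light RGB, plus a rise-and-fade alpha curve, feeding a light renderer with no sprite or mesh.
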NiagaraSystem /BRCosmetics/Weapons/FORT_Melee/Pickaxe_AnglerFemale_1h/FX/NS_Pickaxe_Angler_Idle
dependencies {
 usedDataInterfaces [&NiagaraDataInterfaceColorCurve, &NiagaraDataInterfaceCurve]
 usedDynamicInputs []
 usedModules [&SystemState]
}
emitters {Light {
  emitterUpdateScript [
   &EmitterState()
   &SpawnBurst_Instantaneous(Spawn Count=1, Spawn Probability=HlslExpression() {hlslExpression="Emitter.LoopCount==0"}, Use Spawn Probability=true)
  ]
  particleSpawnScript [
   /Niagara/Modules/Spawn/Initialization/InitializeParticle(
    Module.Write Color=true
    Module.Write Lifetime=true
    Module.Write Mass=false
    Module.Write Material Random=true
    Module.Write Position=false
    Module.Write Scale=false
    Module.Write SpriteRotation=false
    Module.Write SpriteSize=true
    Sprite Size=<75 75>
   )
   &FN_Light_v2(Light Exponent=6, Light Radius=40, LightEnabled=true)
  ]
  particleUpdateScript [
   &ParticleState(Kill Particles When Lifetime Has Elapsed=false, Let Infinitely Lived Particles Die When Emitter Deactivates=true)
   &SkeletalMeshLocation(
    Apply Sampled Orientation to Mesh Particles=true
    Inherit Velocity=false
    MaskSpawnGroup=false
    Mesh Sampling Type=::NewEnumerator5
    Sampled Orientation Coordinate Space=::World
    Sampled Position Coordinate Space=::World
    Sampled Position Offset Coordinate Space:ENiagaraMeshTransforms=::NewEnumerator1
    Sampled Velocity Coordinate Space:ENiagaraCoordinateSpace=::World
    Skeletal Mesh=$User.SkeletalMesh
    Socket Sampling Mode=::NewEnumerator5
   )
   /SetVariables(Particles.GradientColor:Vector3f=&MakeVectorFromLinearColorRGB(Color=&MakeLinearColorFromVectorAndFloat(
      "Vector (RGB)"=&MakeVectorFromLinearColorRGB(Color=&ColorFromCurve(
        ColorCurve=DataInterface() {propertyValues {
          AlphaCurve {keys [<at -0.0023 1 RCIM_Cubic>,<at -0.0016 1>,<at 0.9977 1 RCIM_Cubic>]}
          BlueCurve {keys [
            <at 0 0>
            <at 0.2 0>
            <at 0.4 0>
            <at 0.6 1>
            {arriveTangent=-2.5, interpMode=RCIM_Cubic, leaveTangent=-2.5, time=0.8, value=1.5678}
            <at 1 0>
           ]}
          GreenCurve {keys [
            <at 0 0.039>
            <at 0.2 0.8559>
            <at 0.4 1>
            <at 0.6 0.6207>
            {arriveTangent=-1.4544, interpMode=RCIM_Cubic, leaveTangent=-1.4544, time=0.8, value=0}
            <at 1 0.039>
           ]}
          RedCurve {keys [
            <at 0 1>
            <at 0.2 2>
            <at 0.4 0.3535>
            <at 0.6 0.0629>
            {arriveTangent=2.3427, interpMode=RCIM_Cubic, leaveTangent=2.3427, time=0.8, value=2}
            <at 1 1>
           ]}
         }}
        CurveIndex=$System.NormalizedLoopAge
       ))
      "Float (Alpha)"=1
     ))) {isSetParameters=true}
   /SetVariables(Particles.Light Color:LinearColor=&MakeLinearColorFromVectorAndFloat(
     "Vector (RGB)"=&Multiply_VectorByFloat(Vector=$Particles.GradientColor, Float=1.5)
     "Float (Alpha)"=&FloatFromCurve(
      FloatCurve=DataInterface() {propertyValues {Curve {keys [
          {interpMode=RCIM_Cubic, tangentMode=RCTM_User, time=-0.0063, value=0.0184}
          {arriveTangent=-0.8483, interpMode=RCIM_Cubic, leaveTangent=-0.8483, tangentMode=RCTM_User, time=0.1718, value=1.7147}
          {arriveTangent=-5.0978, interpMode=RCIM_Cubic, leaveTangent=-5.0978, tangentMode=RCTM_User, time=0.2818, value=1.2143}
          {arriveTangent=-0.2783, interpMode=RCIM_Cubic, leaveTangent=-0.2783, tangentMode=RCTM_User, time=0.4294, value=1.1587}
          {arriveTangent=0.3112, interpMode=RCIM_Cubic, leaveTangent=0.3112, tangentMode=RCTM_User, time=0.5407, value=1.3419}
          {arriveTangent=-3.7119, interpMode=RCIM_Cubic, leaveTangent=-3.7119, tangentMode=RCTM_User, time=0.7242, value=0.3749}
          {interpMode=RCIM_Cubic, tangentMode=RCTM_User, time=1.0011, value=0.002}
         ]}}}
      CurveIndex=$System.NormalizedLoopAge
      Scale Curve=0.8
     )
    )) {isSetParameters=true}
  ]
  rendererClasses [&NiagaraLightRendererProperties]
  renderers [{rendererClass=&NiagaraLightRendererProperties, rendererIndex=0}]
 }}
summary {
 systemName=NS_Pickaxe_Angler_Idle
 emitters [{emitterName=Light, rendererClasses [&NiagaraLightRendererProperties]}]
 userVariables [
  User.bIsFrontEndPreview {type=NiagaraBool}
  User.PlayerSpeed {type=NiagaraFloat, defaultValue=0, description=""}
  User.SkeletalMesh {
   type=NiagaraDataInterfaceSkeletalMesh
   defaultValue {
    struct=NiagaraExt_VariableValue_DataInterface
    value {dataInterface=&self:NiagaraDataInterfaceSkeletalMesh_0}
   }
  }
 ]
}
