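# Physics asset for the dynamic fins of the one-handed female Angler pickaxe mesh, with four fins on each side.
# Each fin bone carries a capsule body offset along X, paired with a tiny capsule on its matching MSTR bone.
# One angular-limited constraint links each fin only to its own MSTR body, with modest swing and tight twist, so fins are not chained to one another.
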PhysicsAsset /BRCosmetics/Weapons/FORT_Melee/Pickaxe_AnglerFemale_1h/Meshes/Angler_Female_1h_dyn_fins
bodies {
 dyn_fins_l_01 [{shapeType=Capsule, length=8.4358, radius=2, center=<6 0 0>, rotation=<rot 90 0 0>}]
 dyn_fins_l_01_MSTR [{shapeType=Capsule, length=0.5, radius=0.5}]
 dyn_fins_l_02 [{shapeType=Capsule, length=5.2482, radius=2, center=<5 0 0>, rotation=<rot 90 0 0>}]
 dyn_fins_l_02_MSTR [{shapeType=Capsule, length=0.5, radius=0.5}]
 dyn_fins_l_03 [{shapeType=Capsule, length=3.844, radius=2, center=<4 0 0>, rotation=<rot 90 0 0>}]
 dyn_fins_l_03_MSTR [{shapeType=Capsule, length=0.5, radius=0.5}]
 dyn_fins_l_04 [{shapeType=Capsule, length=4.6285, radius=2, center=<3 0 0>, rotation=<rot 90 0 0>}]
 dyn_fins_l_04_MSTR [{shapeType=Capsule, length=0.5, radius=0.5}]
 dyn_fins_r_01 [{shapeType=Capsule, length=8.4387, radius=2, center=<6 0 0>, rotation=<rot 90 0 0>}]
 dyn_fins_r_01_MSTR [{shapeType=Capsule, length=0.5, radius=0.5}]
 dyn_fins_r_02 [{shapeType=Capsule, length=5.2681, radius=2, center=<5 0 0>, rotation=<rot 90 0 0>}]
 dyn_fins_r_02_MSTR [{shapeType=Capsule, length=0.5, radius=0.5}]
 dyn_fins_r_03 [{shapeType=Capsule, length=3.8777, radius=2, center=<4 0 0>, rotation=<rot 90 0 0>}]
 dyn_fins_r_03_MSTR [{shapeType=Capsule, length=0.5, radius=0.5}]
 dyn_fins_r_04 [{shapeType=Capsule, length=4.3021, radius=2, center=<3 0 0>, rotation=<rot 90 0 0>}]
 dyn_fins_r_04_MSTR [{shapeType=Capsule, length=0.5, radius=0.5}]
}
constraints [
 {bone1Name=dyn_fins_r_01, bone2Name=dyn_fins_r_01_MSTR, swing1LimitDegrees=15, swing2LimitDegrees=25, twistLimitDegrees=5}
 {bone1Name=dyn_fins_r_02, bone2Name=dyn_fins_r_02_MSTR, swing1LimitDegrees=15, swing2LimitDegrees=25, twistLimitDegrees=5}
 {bone1Name=dyn_fins_r_03, bone2Name=dyn_fins_r_03_MSTR, swing1LimitDegrees=15, swing2LimitDegrees=25, twistLimitDegrees=5}
 {bone1Name=dyn_fins_r_04, bone2Name=dyn_fins_r_04_MSTR, swing1LimitDegrees=15, swing2LimitDegrees=25, twistLimitDegrees=5}
 {bone1Name=dyn_fins_l_01, bone2Name=dyn_fins_l_01_MSTR, swing1LimitDegrees=15, swing2LimitDegrees=25, twistLimitDegrees=5}
 {bone1Name=dyn_fins_l_02, bone2Name=dyn_fins_l_02_MSTR, swing1LimitDegrees=15, swing2LimitDegrees=25, twistLimitDegrees=5}
 {bone1Name=dyn_fins_l_03, bone2Name=dyn_fins_l_03_MSTR, swing1LimitDegrees=15, swing2LimitDegrees=25, twistLimitDegrees=5}
 {bone1Name=dyn_fins_l_04, bone2Name=dyn_fins_l_04_MSTR, swing1LimitDegrees=15, swing2LimitDegrees=25, twistLimitDegrees=5}
]
