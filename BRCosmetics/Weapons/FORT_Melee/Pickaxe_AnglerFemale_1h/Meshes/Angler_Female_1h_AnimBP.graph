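# Animation Blueprint for the Angler Female one-handed pickaxe that adds secondary physics motion, using a fins physics override asset.
# The input pose runs through a RigidBody node simulated in b_AR_Root space with reduced gravity and scaled component motion, then goes to output.
# A procedural sway setup, time-based sine rotators feeding additive bone transforms on left and right fins, is left disconnected from both update event and output pose.
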AnimBlueprint /BRCosmetics/Weapons/FORT_Melee/Pickaxe_AnglerFemale_1h/Meshes/Angler_Female_1h_AnimBP
target_skeleton=&/Angler_Female_1h_Axe_Skeleton
graphs [
 AnimGraph {nodes [
   n1 = Get RotateNoiseA() -> (RotateNoiseA:Rotator)
   n2 = "Transform (Modify) Bone - Bone: dyn_fins_r_01_MSTR"(Alpha=1, Translation="0,0,0", Rotation=@n1, Scale="1,1,1") {rotationMode=BMM_Additive, rotationSpace=BCS_BoneSpace, alphaBoolBlend {blendOption=Linear}, boneToModify {boneName=dyn_fins_r_01_MSTR}}
   n3 = Get RotateNoiseB() -> (RotateNoiseB:Rotator)
   n4 = "Transform (Modify) Bone - Bone: dyn_fins_r_02_MSTR"(ComponentPose=@n2, Alpha=1, Translation="0,0,0", Rotation=@n3, Scale="1,1,1") {rotationMode=BMM_Additive, rotationSpace=BCS_BoneSpace, alphaBoolBlend {blendOption=Linear}, boneToModify {boneName=dyn_fins_r_02_MSTR}}
   n5 = Get RotateNoiseC() -> (RotateNoiseC:Rotator)
   n6 = "Transform (Modify) Bone - Bone: dyn_fins_r_03_MSTR"(ComponentPose=@n4, Alpha=1, Translation="0,0,0", Rotation=@n5, Scale="1,1,1") {rotationMode=BMM_Additive, rotationSpace=BCS_BoneSpace, alphaBoolBlend {blendOption=Linear}, boneToModify {boneName=dyn_fins_r_03_MSTR}}
   n7 = Get RotateNoiseD() -> (RotateNoiseD:Rotator)
   n8 = "Transform (Modify) Bone - Bone: dyn_fins_r_04_MSTR"(ComponentPose=@n6, Alpha=1, Translation="0,0,0", Rotation=@n7, Scale="1,1,1") {rotationMode=BMM_Additive, rotationSpace=BCS_BoneSpace, alphaBoolBlend {blendOption=Linear}, boneToModify {boneName=dyn_fins_r_04_MSTR}}
   n9 = "Transform (Modify) Bone - Bone: dyn_fins_l_01_MSTR"(ComponentPose=@n8, Alpha=1, Translation="0,0,0", Rotation=@n1, Scale="1,1,1") {rotationMode=BMM_Additive, rotationSpace=BCS_BoneSpace, alphaBoolBlend {blendOption=Linear}, boneToModify {boneName=dyn_fins_l_01_MSTR}}
   n10 = "Transform (Modify) Bone - Bone: dyn_fins_l_02_MSTR"(ComponentPose=@n9, Alpha=1, Translation="0,0,0", Rotation=@n3, Scale="1,1,1") {rotationMode=BMM_Additive, rotationSpace=BCS_BoneSpace, alphaBoolBlend {blendOption=Linear}, boneToModify {boneName=dyn_fins_l_02_MSTR}}
   n11 = "Transform (Modify) Bone - Bone: dyn_fins_l_03_MSTR"(ComponentPose=@n10, Alpha=1, Translation="0,0,0", Rotation=@n5, Scale="1,1,1") {rotationMode=BMM_Additive, rotationSpace=BCS_BoneSpace, alphaBoolBlend {blendOption=Linear}, boneToModify {boneName=dyn_fins_l_03_MSTR}}
   "Transform (Modify) Bone - Bone: dyn_fins_l_04_MSTR"(ComponentPose=@n11, Alpha=1, Translation="0,0,0", Rotation=@n7, Scale="1,1,1") {rotationMode=BMM_Additive, rotationSpace=BCS_BoneSpace, alphaBoolBlend {blendOption=Linear}, boneToModify {boneName=dyn_fins_l_04_MSTR}}
   n12 = RigidBody(ComponentPose=Local To Component(LocalPose=Input Pose())) {
    bClampLinearTranslationLimitToRefPose=true
    bOverrideWorldGravity=true
    overridePhysicsAsset=&/Angler_Female_1h_dyn_fins
    simulationSpace=BaseBoneSpace
    alphaBoolBlend {blendOption=Linear}
    baseBoneRef {boneName=b_AR_Root}
    componentAppliedLinearAccClamp=<500 500 500>
    componentLinearAccScale=<0.2 0.2 0.2>
    componentLinearVelScale=<0.5 0.5 0.5>
    overrideWorldGravity {z=-10}
   }
   Output Pose(Result=Component To Local(ComponentPose=@n12)) {layerGroup=None}
  ]}
 EventGraph {nodes [
   n1 = Noise Rotate!(execute, self, V:Vector="3,8,10", S:Vector="1,1,1.5", O:Vector="0, 0, 0") -> (then, Rotator:Rotator)
   > Set RotateNoiseA(RotateNoiseA:Rotator=@n1) -> (Output_Get:Rotator)
   > n2 = Noise Rotate(V="3,6,6", S="1,1,1.5", O="0,0.5,1")
   > Set RotateNoiseB(RotateNoiseB:Rotator=@n2) -> (Output_Get:Rotator)
   > n3 = Noise Rotate(V="3,6,4", S="1,1,1.5", O="0,1,2")
   > Set RotateNoiseC(RotateNoiseC:Rotator=@n3) -> (Output_Get:Rotator)
   > n4 = Noise Rotate(V="0,4,4", S="1,1,1.5", O="0,1,2")
   > Set RotateNoiseD(RotateNoiseD:Rotator=@n4) -> (Output_Get:Rotator)
   Event BlueprintUpdateAnimation()
  ]}
 Noise Rotate {nodes [
   n1 = Noise Rotate!() -> (then, V:Vector=~, S:Vector=~, O:Vector=~) {is=title K2Node_FunctionEntry}
   n2 = BreakVector(InVec=vector + float!(A:Vector=@n1.O, B:double=GetGameTimeInSeconds()) -> (ReturnValue:Vector))
   n3 = BreakVector(InVec=@n1.S)
   n4 = "Sin (Radians)"(A=float * float(A=@n2.Z, B=@n3.Z))
   n5 = BreakVector(InVec=@n1)
   n6 = "Sin (Radians)"(A=float * float(A=@n2, B=@n3))
   n7 = "Sin (Radians)"(A=float * float(A=@n2.Y, B=@n3.Y))
   > Return Node!(
    execute
    Rotator:Rotator=MakeRotator(Roll=float * float(A=@n6, B=@n5), Pitch=float * float(A=@n7, B=@n5.Y), Yaw=float * float(A=@n4, B=@n5.Z))
   )
  ]}
]
variables [
 RotateNoiseA {type=Rotator}
 RotateNoiseB {type=Rotator}
 RotateNoiseC {type=Rotator}
 RotateNoiseD {type=Rotator}
]
